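# Translucent, two-sided material that animates instanced cards as drifting, blowing leaves in WildEstate.
# A rotating SubUV petal texture is offset and rounded for masks, colored by the grass function, time-of-day tint and vertex-driven hue shift.
# World position offset combines sprite placement, looping directional travel, sine drift, axis rotation and curl noise, with extra noise dropped on cheaper paths.
# Camera fades shrink and hide leaves; optional wind gusts add displacement and modulate opacity.
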Material /WildEstate/Foliage/Grass/BlowingLeavea/M_WildEstate_FallingLeaves
graph {
 properties {
  bUsedWithInstancedStaticMeshes=true
  bUsedWithNanite=true
  blendMode=BLEND_Translucent
  pixelDepthOffsetMode=PDOM_Legacy
  twoSided=true
  usedShadingModels=Default Lit
 }
 nodes [
  n1 = ScalarParameter() {defaultValue=8, parameterName="U_Images (S)"}
  n2 = ScalarParameter() {defaultValue=4, parameterName="Time Speed (S) Slow"}
  n3 = ScalarParameter() {defaultValue=4, parameterName="Time Speed (S)"}
  n4 = LinearInterpolate(
   A=ScalarParameter() {parameterName=Vertex Alpha Low}
   B=ScalarParameter() {defaultValue=1, parameterName=Vertex Alpha High}
   Alpha=VertexColor().A
  )
  n5 = TextureCoordinate()
  n6 = VectorParameter() {parameterName="Rotation Center (V2)", defaultValue=<rgba 0.5 0.5 0 0>}
  n7 = Add(A=NamedRerouteUsage(), B=Add(A=@n3, B=@n4))
  n8 = TextureSampleParameter2D(UVs=MaterialFunctionCall!(
    U_Images=@n1
    V_Images=@n1
    SubUV Texture RGB
    Time Speed=LinearInterpolate(A=@n2, B=@n3, Alpha=@n4)
    Manual Time Control
    Time=NamedRerouteUsage()
    UVs=MaterialFunctionCall(UVs=@n5, Rotation Center=@n6.RGB, "Rotation Angle (0-1)"=@n7) {materialFunction=&CustomRotator}.Rotated Values
   ) {materialFunction=/Game/Effects/Fort_Effects/Materials/_Functions/MF_SubUV}.UVs) {parameterName=Texture, texture=&/T_toonPetals_basecolor}
  n9 = Add(A=@n8.G, B=ScalarParameter() {defaultValue=0.25, parameterName=SDF Offset Alpha})
  n10 = TextureCoordinate()
  n11 = Constant() {r=0.5}
  n12 = Constant() {r=1}
  n13 = Multiply(
   A=Multiply(A=Round(None=@n9), B=SphereMask(A=@n10, B=@n11, Radius=@n11, Hardness=@n12))
   B=ScalarParameter() {defaultValue=1, parameterName=OpacityMult}
  )
  n14 = Saturate(None=@n13)
  n15 = MaterialFunctionCall!(FadeOutTIme, ResetFadeInTime, FadeInTime) {materialFunction=/Game/Environments/TechArt/WindGustSystem/MF_WindGustsSystem}
  n16 = LinearInterpolate(B=ScalarParameter() {defaultValue=1, parameterName=MaxWindGusts_Opactiy}, Alpha=Clamp(None=Multiply(A=@n15.ResetFadeIn, B=@n15.ResetFadeIn))) {constA=1}
  Multiply(
   A=StaticSwitchParameter(True=Multiply(A=@n14, B=@n16), False=Saturate(None=@n13)) {parameterName=UseWindSystem?}
   B=MaterialFunctionCall(
    Fade Length=ScalarParameter() {defaultValue=250, parameterName="opac Fade Length (S)"}
    Fade Offset=ScalarParameter() {defaultValue=250, parameterName=Opac Fade length}
   ) {materialFunction=&CameraDepthFade}.Result
  )
  n17 = LinearInterpolate(
   A=ScalarParameter() {parameterName=Vertex Red Low}
   B=ScalarParameter() {defaultValue=1, parameterName=Vertex Red High}
   Alpha=VertexColor().R
  )
  n18 = ScalarParameter() {defaultValue=-0.2, parameterName=HueShiftAmount}
  n19 = GetMaterialAttributes(None=MaterialFunctionCall!(NormalOverrideWS, UseCamera facing BildBoards, UseLandscapeCoords) {materialFunction=/WildEstate/Environment/BaseMaterials/MaterialFunctions/MF_WildEstate_Grass}.Out Material) {attributeGetTypes [#1]}
  n20 = MaterialFunctionCall(
   Color_Morning=VectorParameter() {parameterName="Color_Morning (V3)", defaultValue=<rgba 1 1 1 0>}.RGB
   Color_Day=VectorParameter() {parameterName="Color_Day (V3)", defaultValue=<rgba 1 1 1 0>}.RGB
   Color_Evening=VectorParameter() {parameterName="Color_Evening (V3)", defaultValue=<rgba 1 1 1 0>}.RGB
   Color_Night=VectorParameter() {parameterName="Color_Night (V3)", defaultValue=<rgba 1 1 1 0>}.RGB
  ) {materialFunction=&Color_Time_of_Day}
  n21 = ScalarParameter() {defaultValue=0.25, parameterName=SDF Offset Red}
  Multiply(
   A=MaterialFunctionCall(Hue Shift Percentage=Multiply(A=@n17, B=@n18) {constB=-0.2}, Texture=Multiply(A=@n19.Base Color, B=@n20.Result)) {materialFunction=&HueShift}.Result
   B=Round(None=Add(A=@n8.R, B=@n21))
  )
  NamedRerouteDeclaration(None=Multiply(A=Time(), B=ScalarParameter() {defaultValue=1000, parameterName=MasterTime})) {name=M Time, nodeColor=<rgba 0.1016 0 1 1>}
  n22 = Multiply(
   A=MaterialFunctionCall(In=@n15.LinearTimeProgressionNormalized, Power=ScalarParameter() {defaultValue=3, parameterName=WindGustsAccelaration}) {materialFunction=&SCurve}.Result
   B=ScalarParameter() {defaultValue=1500, parameterName=WindGustsDistanceOffset}
  )
  n23 = MaterialFunctionCall!(Position, IsTrees, IsGrass, IsMediumPlant, IsMediumPlant_2024Anim, IsMediumPlant2ndLevel, MediumPlant_BasePivotPostion) {materialFunction=/Game/Environments/Asteria/Foliage/Wind/MF_WindActorData}
  n24 = ScalarParameter() {defaultValue=1, parameterName=SizeX}
  n25 = Multiply(
   A=ScalarParameter() {defaultValue=0.25, parameterName=MinSizeMult}
   B=LinearInterpolate(
    A=ScalarParameter() {parameterName=Vertex Green Low}
    B=ScalarParameter() {defaultValue=1, parameterName=Vertex Green High}
    Alpha=VertexColor().G
   )
  )
  n26 = ScalarParameter() {defaultValue=500, parameterName=maxSize}
  n27 = ComponentMask(None=TextureCoordinate() {coordinateIndex=1}) {g=true}
  n28 = Add(A=Divide(A=NamedRerouteUsage()) {constB=3}, B=Multiply(A=Subtract(A=@n27) {constB=0.5}) {constB=-2})
  n29 = Sine(None=Add(A=@n28) {constB=-0.25})
  n30 = Power(Base=LinearInterpolate(Alpha=ConstantBiasScale(None=@n29))) {constExponent=0.6}
  n31 = ScalarParameter() {defaultValue=50, parameterName=FadeDistance}
  n32 = Constant() {r=1}
  n33 = ShadingPathSwitch(Default=QualitySwitch(Default=DepthFade(FadeDistance=@n31), Low=@n32), Mobile=@n32)
  n34 = Multiply(
   A=LinearInterpolate(A=Multiply(A=@n25, B=@n26), B=@n26, Alpha=Multiply(A=@n30, B=@n33))
   B=MaterialFunctionCall(
    Fade Length=ScalarParameter() {defaultValue=500, parameterName=NearCamShrinkDist}
    Fade Offset=ScalarParameter() {defaultValue=500, parameterName=NearCamShrinkOffset}
   ) {materialFunction=&CameraDepthFade}.Result
  )
  n35 = ScalarParameter() {defaultValue=1, parameterName=SizeY}
  n36 = MaterialFunctionCall!(
   Center Location=WorldPosition().XYZ
   0-1 UVs
   XY scale=AppendVector(A=Multiply(A=@n24, B=@n34), B=Multiply(A=@n35, B=@n34))
   OverridePivot=StaticBoolParameter() {parameterName="OverridePivot (SB)"}
   "Normals (Optional)"
  ) {materialFunction=&Sprite}
  n37 = Multiply(A=NamedRerouteUsage(), B=ScalarParameter() {defaultValue=0.001, parameterName=VarianceSpeed}) {constB=0.05}
  n38 = ComponentMask(None=TextureCoordinate() {coordinateIndex=1}) {r=true}
  n39 = Add(
   A=Multiply(A=NamedRerouteUsage(), B=ScalarParameter() {defaultValue=0.4, parameterName=DriftSpeed})
   B=Sine(None=Add(A=@n37, B=@n38)) {period=0.17}
  )
  n40 = AppendVector(A=Sine(None=@n39) {period=0.56}, B=Sine(None=@n39) {period=0.7})
  n41 = Sine(None=@n39) {period=0.82}
  n42 = ScalarParameter() {defaultValue=1, parameterName=MinHorizontalDrift}
  n43 = ScalarParameter() {defaultValue=1, parameterName=MaxHorizontalDrift}
  n44 = Sine(None=Add(A=@n37, B=@n38)) {period=0.14}
  n45 = Multiply(A=ObjectRadius(), B=Constant() {r=0.5}) {constB=0.5}
  n46 = LinearInterpolate(A=Multiply(A=@n45) {constB=-1}, B=@n45, Alpha=Frac(None=@n28))
  n47 = ScalarParameter() {defaultValue=1, parameterName=OverallSpeed}
  n48 = Add(
   A=Multiply(A=AppendVector(A=@n40, B=@n41), B=LinearInterpolate(A=@n42, B=@n43, Alpha=@n44))
   B=Multiply(A=Multiply(A=@n46, B=@n47), B=VectorParameter() {parameterName=OverallDirection, defaultValue=<rgba 1 0 0 1>}.RGB)
  )
  n49 = ScalarParameter() {defaultValue=50, group=WPO Curl, parameterName=CurIntensity}
  n50 = ScalarParameter() {defaultValue=100, group=WPO Curl, parameterName=CurIntensityHigh}
  n51 = VertexColor()
  n52 = Multiply(A=Multiply(A=ScalarParameter() {defaultValue=1, parameterName=Curl}, B=LinearInterpolate(A=@n49, B=@n50, Alpha=@n51.A)), B=VertexColor().A)
  n53 = Subtract(A=WorldPosition().XYZ, B=ActorPositionWS())
  n54 = Multiply(A=NamedRerouteUsage(), B=ScalarParameter() {defaultValue=1, group=WPO Curl, parameterName=CurlSpeed})
  n55 = LinearInterpolate(
   A=ScalarParameter() {defaultValue=1000, group=WPO Curl, parameterName=CurlScale}
   B=ScalarParameter() {defaultValue=500, group=WPO Curl, parameterName=CurlScale_Low}
   Alpha=@n51.A
  )
  n56 = ScalarParameter() {defaultValue=1, parameterName=Curl}
  n57 = VectorNoise(World Position=Divide(A=Subtract(A=@n53, B=@n54), B=Divide(A=@n55, B=@n56))) {noiseFunction=VNF_VectorALU}
  n58 = Transform(None=VectorParameter() {parameterName=RotationAxis, defaultValue=<rgba 0 0 1 0>}.RGB) {transformSourceType=TRANSFORMSOURCE_Local}
  n59 = ObjectPositionWS()
  n60 = Add(A=NamedRerouteUsage(), B=ComponentMask(None=@n59) {b=true})
  n61 = ObjectPositionWS()
  n62 = LinearInterpolate(
   A=ScalarParameter() {defaultValue=1, parameterName=RotationSpeed}
   B=ScalarParameter() {defaultValue=1, parameterName=RotationSpeed_Low}
   Alpha=Frac(None=ComponentMask(None=@n61) {b=true})
  )
  n63 = Add(
   A=Add(A=Add(A=@n36.World position offset, B=@n48), B=Multiply(A=@n52, B=@n57))
   B=RotateAboutAxis(NormalizedRotationAxis=Normalize(VectorInput=@n58), RotationAngle=Multiply(A=@n60, B=@n62), PivotPoint=@n59, Position=WorldPosition().XYZ)
  )
  n64 = ScalarParameter() {defaultValue=50, group=WPO Curl, parameterName=CurIntensity}
  n65 = ScalarParameter() {defaultValue=100, group=WPO Curl, parameterName=CurIntensityHigh}
  n66 = LinearInterpolate(
   A=ScalarParameter() {parameterName=Vertex Alpha Low}
   B=ScalarParameter() {defaultValue=1, parameterName=Vertex Alpha High}
   Alpha=VertexColor().A
  )
  n67 = Multiply(A=Multiply(A=ScalarParameter() {defaultValue=1, parameterName=Curl}, B=LinearInterpolate(A=@n64, B=@n65, Alpha=@n66)), B=@n66)
  n68 = Subtract(A=WorldPosition().XYZ, B=ActorPositionWS())
  n69 = Multiply(
   A=NamedRerouteUsage()
   B=Multiply(A=ScalarParameter() {defaultValue=1, group=WPO Curl, parameterName=CurlSpeed}, B=ScalarParameter() {defaultValue=1, parameterName=Curl})
  )
  n70 = LinearInterpolate(
   A=ScalarParameter() {defaultValue=1000, group=WPO Curl, parameterName=CurlScale}
   B=ScalarParameter() {defaultValue=500, group=WPO Curl, parameterName=CurlScale_Low}
   Alpha=@n66
  )
  n71 = ScalarParameter() {defaultValue=1, parameterName=Curl 2}
  n72 = VectorNoise(World Position=Divide(A=Subtract(A=@n68, B=@n69), B=Divide(A=@n70, B=@n71))) {noiseFunction=VNF_VectorALU}
  n73 = Add(A=@n63, B=StaticSwitchParameter(True=Multiply(A=@n67, B=@n72), False=Constant()) {defaultValue=true, parameterName=UseCurlNoise?})
  n74 = QualitySwitch(
   Default=ShadingPathSwitch(Default=ShadingPathSwitch(Default=@n73, Mobile=@n63) {desc=Migrated from FeatureLevelSwitch}, Forward=@n63, Mobile=@n63)
   Low=@n63
  )
  StaticSwitchParameter(True=Add(A=Multiply(A=@n22, B=@n23.WindI_Direction), B=@n74), False=@n74) {parameterName=UseWindSystem?}
 ]
}
